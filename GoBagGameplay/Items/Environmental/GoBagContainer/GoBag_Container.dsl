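; GoBag container Blueprint, an environmental searchable item that behaves as a chest in the game.
; The construction script calls its parent, then fills the custom Niagara effects array with the custom chest aura as its sole entry.
; It destroys itself when the actor it is linked to beneath it is destroyed (FORT-631061).
; BeginPlay, overlap and Tick are unimplemented stubs.
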(event EventBeginPlay
  (|CallParentFunction) ; This node is disabled and will not be called.\nDrag off pins to build functionality.
)

(event Collision|EventActorBeginOverlap
  (params (OtherActor :type Actor%20Object%20Reference)))

(event EventTick
  (params (DeltaSeconds :type Float%20%28single-precision%29)))

(event OnLinkedActorDestroyed(FortLinkToActor)
  (params (DamageCauser :type Actor%20Object%20Reference))
  ;[1 1 1] FORT-631061: This will link this actor to what's immediately below it when it's not a floor. When it's link is destroyed, destroy the chest.
  (Actor|DestroyActor))

(fn ConstructionScript
  (|CallParentFunction)
  (Variables|Default|SetCustom%20Niagara%20Effects (Utilities|Array|MakeArray (Variables|Default|GetCustom%20Chest%20Aura))))
