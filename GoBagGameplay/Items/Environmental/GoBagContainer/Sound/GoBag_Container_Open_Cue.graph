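# Sound cue for opening a GoBag container, layering a main open sound with a delayed sweetener.
# A mixer blends a random pick of three open waves at slightly raised pitch with a random pick of three sweetener waves.
# The sweetener branch starts after a short fixed delay, pitched higher with slight variation and played quieter.
# It uses item-interact attenuation and limits playback to one instance per owner, blocking new triggers.
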SoundCue /GoBagGameplay/Items/Environmental/GoBagContainer/Sound/GoBag_Container_Open_Cue
root=Mixer_0
nodes [
 {id=Delay_0, children [Modulator_3], properties {delayMax=0.1, delayMin=0.1}}
 {id=Mixer_0, children [Modulator_1,Delay_0], properties {inputVolume [2,0.5]}}
 {id=Modulator_1, children [Random_0], properties {pitchMax=1.1, pitchMin=1.1, volumeMax=1, volumeMin=1}}
 {id=Modulator_3, children [Random_6], properties {pitchMax=1.8, pitchMin=1.6, volumeMax=0.7, volumeMin=0.7}}
 {id=Random_0, children [WavePlayer_4,WavePlayer_5,WavePlayer_6], properties {weights [1,1,1]}}
 {id=Random_6, children [WavePlayer_7,WavePlayer_8,WavePlayer_9], properties {bShouldExcludeFromBranchCulling=true, weights [1,1,1]}}
 {id=WavePlayer_4, properties {soundWaveAssetPtr=&/GoBag_Open_01}}
 {id=WavePlayer_5, properties {soundWaveAssetPtr=&/GoBag_Open_02}}
 {id=WavePlayer_6, properties {soundWaveAssetPtr=&/GoBag_Open_03}}
 {id=WavePlayer_7, properties {soundWaveAssetPtr=&/GoBag_Open_Sweetener_01}}
 {id=WavePlayer_8, properties {soundWaveAssetPtr=&/GoBag_Open_Sweetener_02}}
 {id=WavePlayer_9, properties {soundWaveAssetPtr=&/GoBag_Open_Sweetener_03}}
]
properties {
 attenuationSettings=/Game/Sounds/Attenuation/Fort_Item_Interact_Att
 bOverrideConcurrency=true
 maxDistance=1850
 soundClassObject=&Fort_Loot_Reward_PSM
 volumeMultiplier=2
 concurrencyOverrides {bLimitToOwner=true, maxCount=1, resolutionRule=PreventNew}
}
